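; Frameless Dragon Cart smashing door trap that, when activated, crushes players and vehicles overlapping its kill box.
; After a short delay, TL_SmashDoor drives the door percentage, with a server-only smash on the forward pass; reset plays it in reverse.
; MoveDoorByPercentage lerps panel height and Z scale and toggles the blocking box's collision.
; Health comes from a scalable float; death drops metal and spawns bounds- and hit-aware Niagara debris.
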(event EventBeginPlay
  (|CallParentFunction)
  (Collision|SetCollisionEnabled (Variables|Default|GetBlockingBox) "NoCollision")
  (bind _actorcenterpoint (CallFunction|GetActorCenterPoint))
  (Variables|Default|SetActorCenterLocation _actorcenterpoint))

(event EventBP_OnActive
  (params (EnableAlternateActive :type Boolean) (EnableOpenCues :type Boolean))
  (|CallParentFunction EnableAlternateActive EnableOpenCues)
  (Audio|PlaySoundatLocation "/DragonCartDoors/Assets/Audio/SC_DragonCartDoors_Trap_Cue.SC_DragonCartDoors_Trap_Cue" (Transformation|GetActorLocation))
  (Utilities|FlowControl|Delay 0.125)
  (Utilities|FlowControl|Sequence
    (:then_0
      (bind (_direction _panelheight) (Timeline|TL_SmashDoor SetNewTime :NewTime 0.025)
        (:Update
          (Variables|Default|SetDoorOperationPercentage _panelheight))
        (:Finished
          (CallFunction|OnDoorTimelineFinished _direction))
        (:ShouldSmash
          (switch Utilities|FlowControl|Switch|SwitchonETimelineDirection _direction
            (:Forward
              (Networking|SwitchHasAuthority
                (:Authority
                  (CallFunction|SmashPlayers (CallFunction|GetPlayerstoSmash))
                  (CallFunction|SmashVehicles (CallFunction|GetVehiclestoSmash)))))))))
    (:then_1
      (Timeline|TL_SmashDoor Play))))

(event Default|EventBP_OnReset
  (params (EnableAlternateReset :type Boolean) (EnableOpenCues :type Boolean))
  (|CallParentFunction EnableAlternateReset EnableOpenCues)
  (Audio|PlaySoundatLocation "/Game/Sounds/Fort_Doors/World/Door_Garage01_Open_Cue.Door_Garage01_Open_Cue" (Transformation|GetActorLocation))
  (Utilities|FlowControl|Delay 0.125)
  (Timeline|TL_SmashDoor ReverseFromEnd))

(event Custom|AutoReset)

(event EventSetDoorHealthFromDataTable
  (Gameplay|ForceBuildingHealth self (Math|Conversions|ToFloat(Integer) (Math|Float|Truncate (ScalableFloat|GetValueatLevel (Variables|Default|GetScalarDoorHealth))))))

(event EventOnDeathServer
  (params (Damage :type Float%20%28single-precision%29) (DamageTags :type Gameplay%20Tag%20Container%20Structure%20%28by%20ref%29) (Momentum :type Vector :default "0, 0, 0") (HitInfo :type Hit%20Result%20Structure%20%28by%20ref%29) (InstigatedBy :type Controller%20Object%20Reference) (DamageCauser :type Actor%20Object%20Reference) (EffectContext :type Gameplay%20Effect%20Context%20Handle%20Structure))
  (|CallParentFunction Damage DamageTags Momentum HitInfo InstigatedBy DamageCauser EffectContext)
  (Inventory|SpawnPickupInWorld "/Game/Items/ResourcePickups/MetalItemData.MetalItemData" 12 (Variables|Default|GetActorCenterLocation) "0.000000,0.000000,1.000000" 12 true true true -1 "Destruction" "Unset" 0 false))

(event EventOnDeathPlayEffects
  (params (Damage :type Float%20%28single-precision%29) (DamageTags :type Gameplay%20Tag%20Container%20Structure%20%28by%20ref%29) (Momentum :type Vector :default "0, 0, 0") (HitInfo :type Hit%20Result%20Structure%20%28by%20ref%29) (InstigatedBy :type Fort%20Pawn%20Object%20Reference) (DamageCauser :type Actor%20Object%20Reference) (EffectContext :type Gameplay%20Effect%20Context%20Handle%20Structure))
  (bind _panel (Variables|Default|GetPanel))
  (bind (_returnvalue_origin _returnvalue_boxextent _returnvalue_sphereradius) (Math|BoxSphereBounds|BreakBoxSphereBounds (Fortnite|Utility|GetLocalBounds _panel)))
  (bind (_returnvalue_location _returnvalue_rotation) (Math|Transform|BreakTransform (Transformation|GetActorTransform)))
  (bind (_bblockinghit _binitialoverlap _time _distance _location) (Collision|BreakHitResult HitInfo))
  (Audio|PlaySoundatLocation (Variables|Default|GetDestoySound) _location)
  (bind _returnvalue (Niagara|SpawnSystematLocation "/Game/Effects/Fort_Effects/World/Destruction/Destruction2/Custom/NS_LargeDoor_Destruction.NS_LargeDoor_Destruction" (Transformation|GetActorLocation)))
  (Niagara|SetNiagaraVariable(Vector3) _returnvalue "DoorScaling" (* (Variables|/Script/Engine.SceneComponent|GetRelativeScale3D (Variables|Default|GetPanel)) "1.000000,1.000000,2.000000"))
  (Niagara|SetNiagaraVariable(Vector3) _returnvalue "HitVector" (Math|Vector|Normalize (- _location (Transformation|GetActorLocation InstigatedBy))))
  (Niagara|SetNiagaraVariable(Vector3) _returnvalue "HitLocation" _location)
  (Niagara|SetNiagaraVariable(Vector3) _returnvalue "BoundsExtents" _returnvalue_boxextent)
  (Niagara|SetNiagaraVariable(Float) _returnvalue "Radius" _returnvalue_sphereradius)
  (Niagara|SetNiagaraVariable(Quaternion) _returnvalue "Orientation" (Math|Conversions|ToQuaternion(Rotator) _returnvalue_rotation)))

(event OnLinkedActorDestroyed(FortLinkToActor)
  (params (DamageCauser :type Actor%20Object%20Reference))
  (Rendering|SetActorHiddenInGame true)
  (Actor|DestroyActor))

(fn GetActorCenterPoint
  (bind (_origin _boxextent) (Collision|GetActorBounds false))
  (return (Math|Vector|MakeVector (.x (Transformation|GetActorLocation)) (.y (Transformation|GetActorLocation)) (+ (.z (Transformation|GetActorLocation)) (* (.z _boxextent) 0.5)))))

(fn GetPlayersToSmash
  (return (Collision|GetOverlappingActors (Variables|Default|GetKillBox) "/Script/FortniteGame.FortPlayerPawn" :!class /Script/Engine.PrimitiveComponent)))

(fn GetVehiclesToSmash
  (return (Collision|GetOverlappingActors (Variables|Default|GetKillBox) 0 :!class /Script/Engine.PrimitiveComponent)))

(fn MoveDoorByPercentage
  (params (Percentage :type Float%20%28double-precision%29))
  (bind _returnvalue (< Percentage 1.0))
  (bind _returnvalue_2 (select _returnvalue "NoCollision" "QueryAndPhysics"))
  (Transformation|SetRelativeTransform (Variables|Default|GetPanel) (Math|Transform|MakeTransform (Math|Vector|MakeVector (.x (Variables|/Script/Engine.SceneComponent|GetRelativeLocation (Variables|Default|GetPanel))) (.y (Variables|/Script/Engine.SceneComponent|GetRelativeLocation (Variables|Default|GetPanel))) (Math|Float|Lerp (Variables|Default|GetPanelHeightStartZ) (Variables|Default|GetPanelHeightEndZ) Percentage)) "0, 0, 0" (Math|Vector|MakeVector (.x (Variables|/Script/Engine.SceneComponent|GetRelativeScale3D (Variables|Default|GetPanel))) (.y (Variables|/Script/Engine.SceneComponent|GetRelativeScale3D (Variables|Default|GetPanel))) (Math|Float|Lerp (Variables|Default|GetPanelScaleStartZ) (Variables|Default|GetPanelScaleEndZ) Percentage))) false false)
  (Development|PrintString (Utilities|String|ToString(Byte) _returnvalue_2))
  (Collision|SetCollisionEnabled (Variables|Default|GetBlockingBox) _returnvalue_2))

(fn OnDoorTimelineFinished
  (params (Direction :type ETimelineDirection%20Enum :default "Forward"))
  (|CallParentFunction Direction)
  (switch Utilities|FlowControl|Switch|SwitchonETimelineDirection Direction
    (:Forward
      (Collision|SetCollisionEnabled (Variables|Default|GetPanel) "QueryOnly"))
    (:Backward
      (Collision|SetCollisionEnabled (Variables|Default|GetPanel) "QueryAndPhysics"))))

(fn ConstructionScript
  (|CallParentFunction))
