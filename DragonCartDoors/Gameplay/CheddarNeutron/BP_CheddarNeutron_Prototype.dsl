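; Prototype smashing door (CheddarNeutron) that crushes players in its KillBox, extending a parent door Blueprint.
; Activation plays a timeline setting DoorOperationPercentage; MoveDoorByPercentage lerps panel height and scale plus frame height, and forward completion calls SmashPlayers on authority.
; KillBox overlaps debug-print and set a safety light (red/iron via OnRep), which clears on exit when either players or vehicles are absent.
; Linked-actor destruction destroys the door.
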(event EventBP_OnActive
  (params (EnableAlternateActive :type Boolean) (EnableOpenCues :type Boolean))
  (|CallParentFunction EnableAlternateActive EnableOpenCues)
  (bind (_direction _panelheight) (Timeline|TL_SmashDoor PlayFromStart)
    (:Update
      (Variables|Default|SetDoorOperationPercentage _panelheight))
    (:Finished
      (CallFunction|OnDoorTimelineFinished _direction))))

(event OnComponentBeginOverlap(KillBox)
  (params (OverlappedComponent :type Primitive%20Component%20Object%20Reference) (OtherActor :type Actor%20Object%20Reference) (OtherComp :type Primitive%20Component%20Object%20Reference) (OtherBodyIndex :type Integer) (bFromSweep :type Boolean) (SweepResult :type Hit%20Result%20Structure%20%28by%20ref%29))
  (Variables|Default|SetSafteyLightOn true)
  (Development|PrintString (Utilities|GetDisplayName OtherActor) false true "(R=1.000000,G=0.400000,B=0.000000,A=1.000000)" 5.0))

(event OnComponentEndOverlap(KillBox)
  (params (OverlappedComponent :type Primitive%20Component%20Object%20Reference) (OtherActor :type Actor%20Object%20Reference) (OtherComp :type Primitive%20Component%20Object%20Reference) (OtherBodyIndex :type Integer))
  (if (< (Utilities|Array|Length (Collision|GetOverlappingActors (Variables|Default|GetKillBox) "/Script/FortniteGame.FortPlayerPawn" :!class /Script/Engine.PrimitiveComponent)) 1)
    (Variables|Default|SetSafteyLightOn false)
    (elif (< (Utilities|Array|Length (Collision|GetOverlappingActors (Variables|Default|GetKillBox) "/Script/FortniteGame.FortAthenaVehicle" :!class /Script/Engine.PrimitiveComponent)) 1)
      (Variables|Default|SetSafteyLightOn false))))

(event OnLinkedActorDestroyed(FortLinkToActor)
  (params (DamageCauser :type Actor%20Object%20Reference))
  (Rendering|SetActorHiddenInGame true)
  (Actor|DestroyActor))

(fn GetPlayersToSmash
  (return (Collision|GetOverlappingActors (Variables|Default|GetKillBox) "/Script/FortniteGame.FortPlayerPawn" :!class /Script/Engine.PrimitiveComponent)))

(fn GetVehiclesToSmash
  (return (Collision|GetOverlappingActors (Variables|Default|GetKillBox) 0 :!class /Script/Engine.PrimitiveComponent)))

(fn MoveDoorByPercentage
  (params (Percentage :type Float%20%28double-precision%29))
  (|CallParentFunction Percentage)
  (Transformation|SetRelativeTransform (Variables|Default|GetPanel) (Math|Transform|MakeTransform (Math|Vector|MakeVector (.x (Variables|/Script/Engine.SceneComponent|GetRelativeLocation (Variables|Default|GetPanel))) (.y (Variables|/Script/Engine.SceneComponent|GetRelativeLocation (Variables|Default|GetPanel))) (Math|Float|Lerp (Variables|Default|GetPanelHeightStartZ) (Variables|Default|GetPanelHeightEndZ) Percentage)) "0, 0, 0" (Math|Vector|MakeVector (.x (Variables|/Script/Engine.SceneComponent|GetRelativeScale3D (Variables|Default|GetPanel))) (.y (Variables|/Script/Engine.SceneComponent|GetRelativeScale3D (Variables|Default|GetPanel))) (Math|Float|Lerp (Variables|Default|GetPanelScaleStartZ) (Variables|Default|GetPanelScaleEndZ) Percentage))) false false)
  (Transformation|SetRelativeLocation (Variables|Default|GetPanelFrameMesh) (Math|Vector|MakeVector (.x (Variables|/Script/Engine.SceneComponent|GetRelativeLocation (Variables|Default|GetPanelFrameMesh))) (.y (Variables|/Script/Engine.SceneComponent|GetRelativeLocation (Variables|Default|GetPanelFrameMesh))) (Math|Float|Lerp (Variables|Default|GetPanelFrameHeightStartZ) (Variables|Default|GetPanelFrameHeightEndZ) Percentage)) false false))

(fn OnDoorTimelineFinished
  (params (Direction :type ETimelineDirection%20Enum :default "Forward"))
  (|CallParentFunction Direction)
  (switch Utilities|FlowControl|Switch|SwitchonETimelineDirection Direction
    (:Forward
      (Networking|SwitchHasAuthority
        (:Authority
          (CallFunction|SmashPlayers (CallFunction|GetPlayerstoSmash)))))))

(fn OnRep_SafteyLightOn
  (if (Variables|Default|GetSafteyLightOn)
    (Rendering|Material|SetMaterial (Variables|Default|GetSafteyLight) 0 "/DragonCartDoors/Assets/MI_FlatColorParent_Red.MI_FlatColorParent_Red")
    (return)
    (else
      (Rendering|Material|SetMaterial (Variables|Default|GetSafteyLight) 0 "/DragonCartDoors/Assets/MI_FlatColorParent_Iron.MI_FlatColorParent_Iron")
      (return))))

(fn ConstructionScript
  (|CallParentFunction))
